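# Sound cue for the equip sound of the BeastGirlStorm PomPom item, classed as weapon foley.
# It randomly picks one of two equally weighted waves, shared by every volume path.
# Nested branches on the islocal and isenemy parameters set volume: full for local, reduced for others, with enemies slightly louder than non-enemies.
# It references Fortnite item attenuation and concurrency assets.
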SoundCue /BeastGirlStormClawsGameplay/Assets/Audio/SoundCues/PomPom/SC_BeastGirlStorm_PomPom_Equip
root=Branch_0
nodes [
 {id=Branch_0, children [Modulator_2,Branch_3,Modulator_2], properties {boolParameterName=islocal}}
 {id=Branch_3, children [Modulator_4,Modulator_3,Modulator_4], properties {boolParameterName=isenemy}}
 {id=Modulator_2, children [Random_0], properties {pitchMax=1, pitchMin=1, volumeMax=1, volumeMin=1}}
 {id=Modulator_3, children [Random_0], properties {pitchMax=1, pitchMin=1, volumeMax=0.25, volumeMin=0.25}}
 {id=Modulator_4, children [Random_0], properties {pitchMax=1, pitchMin=1, volumeMax=0.3, volumeMin=0.3}}
 {id=Random_0, children [WavePlayer_1,WavePlayer_2], properties {weights [1,1]}}
 {id=WavePlayer_1, properties {soundWaveAssetPtr=/BeastGirlStormClawsGameplay/Assets/Audio/Wavs/PomPom/BeastGirlStorm_PomPom_Equip_02}}
 {id=WavePlayer_2, properties {soundWaveAssetPtr=/BeastGirlStormClawsGameplay/Assets/Audio/Wavs/PomPom/BeastGirlStorm_PomPom_Equip_03}}
]
properties {
 attenuationSettings=/FortMetasound/Items/AS_Fort_Item
 maxDistance=3500
 soundClassObject=&Fort_WeaponFoley
 attenuationOverrides {bAttenuate=false, lPFRadiusMin=17, stereoSpread=1000}
 concurrencyOverrides {maxCount=1}
 concurrencySet [&CS_Fort_Item]
}
