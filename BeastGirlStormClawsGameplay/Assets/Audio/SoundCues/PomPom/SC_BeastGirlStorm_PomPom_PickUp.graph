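# Sound cue for the pick-up sound of the BeastGirlStorm PomPom item.
# A Random node picks one of three equally weighted pick-up wave variations.
# It plays in the loot-collect sound class with the shared Fortnite item attenuation asset.
# A concurrency override caps it at one playing instance, taking the place of the referenced item concurrency set.
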SoundCue /BeastGirlStormClawsGameplay/Assets/Audio/SoundCues/PomPom/SC_BeastGirlStorm_PomPom_PickUp
root=Random_0
nodes [
 {id=Random_0, children [WavePlayer_0,WavePlayer_1,WavePlayer_2], properties {weights [1,1,1]}}
 {id=WavePlayer_0, properties {soundWaveAssetPtr=/BeastGirlStormClawsGameplay/Assets/Audio/Wavs/PomPom/BeastGirlStorm_PomPom_PickUp_01}}
 {id=WavePlayer_1, properties {soundWaveAssetPtr=/BeastGirlStormClawsGameplay/Assets/Audio/Wavs/PomPom/BeastGirlStorm_PomPom_PickUp_02}}
 {id=WavePlayer_2, properties {soundWaveAssetPtr=/BeastGirlStormClawsGameplay/Assets/Audio/Wavs/PomPom/BeastGirlStorm_PomPom_PickUp_03}}
]
properties {
 attenuationSettings=/FortMetasound/Items/AS_Fort_Item
 bOverrideConcurrency=true
 maxDistance=3500
 soundClassObject=&Fort_Loot_Collect
 volumeMultiplier=1
 attenuationOverrides {bAttenuate=false, lPFRadiusMin=17, stereoSpread=1000}
 concurrencyOverrides {maxCount=1}
 concurrencySet [&CS_Fort_Item]
}
